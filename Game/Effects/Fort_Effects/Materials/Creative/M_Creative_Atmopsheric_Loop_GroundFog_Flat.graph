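# Unlit translucent sprite particle material for looping flat ground fog in Creative, tinted by particle color and time-of-day colors.
# Opacity multiplies a random-frame smoke flipbook by two summed, counter-panning inky smoke layers on particle macro UVs, distorted by rotating smoke noise.
# Particle alpha, applied twice, and depth fade shape the final opacity.
# Quality and shading-path switches drop depth fade and simplify noise for Low and Mobile; Low also skips time-of-day tinting.
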Material /Game/Effects/Fort_Effects/Materials/Creative/M_Creative_Atmopsheric_Loop_GroundFog_Flat
graph {
 properties {
  allowTranslucentCustomDepthWrites=true
  bAutomaticallySetUsageInEditor=false
  bEnableResponsiveAA=true
  bTangentSpaceNormal=false
  bUsedWithParticleSprites=true
  bUsedWithStaticMesh=false
  blendMode=BLEND_Translucent
  opacityMaskClipValue=0.01
  pixelDepthOffsetMode=PDOM_Legacy
  shadingModel=MSM_Unlit
  translucencyLightingMode=TLM_VolumetricPerVertexNonDirectional
  translucencyPass=MTP_BeforeDOF
  usedShadingModels=Unlit
 }
 nodes [
  n1 = ParticleColor()
  n2 = MaterialFunctionCall(
   Color_Morning=Multiply(A=VectorParameter() {parameterName=Color_Morning, defaultValue=<rgba 1 0.6565 0.5 0>}.RGB, B=@n1.RGB)
   Color_Day=Multiply(A=VectorParameter() {parameterName=Color_Day, defaultValue=<rgba 1 0.955 0.85 0>}.RGB, B=@n1.RGB)
   Color_Evening=Multiply(A=VectorParameter() {parameterName=Color_Evening, defaultValue=<rgba 0.5 0.7687 1 0>}.RGB, B=@n1.RGB)
   Color_Night=Multiply(A=@n1.RGB, B=VectorParameter() {parameterName=Color_Night, defaultValue=<rgba 0.25 0.3719 0.5 0>}.RGB)
  ) {materialFunction=&Color_Time_of_Day}
  MaterialFunctionCall(Emissive=QualitySwitch(Default=@n2.Result, Low=ParticleColor().RGB, High=@n2.Result, Medium=@n2.Result)) {materialFunction=&MF_FX_EmissiveAdjustments}
  n3 = ScalarParameter() {defaultValue=8, parameterName=RowsAndColumns}
  n4 = MaterialFunctionCall!(UseTexCoord 1, Rows=@n3, Columns=@n3, Particle Random Value=DynamicParameter() {defaultValue=<rgba 0 0 0 8>}.Param3) {materialFunction=/Game/Effects/Fort_Effects/Materials/_Functions/MF_RandomSubUV_GPU_Sprites}
  n5 = TextureCoordinate()
  n6 = QualitySwitch(Default=ShadingPathSwitch(Default=Clamp(None=@n4.Result), Mobile=@n5), Low=@n5)
  n7 = DynamicParameter() {defaultValue=<rgba 0 0 0 8>}
  n8 = ParticleMacroUV()
  n9 = Panner(Coordinate=@n8) {speedX=0.02, speedY=-0.015}
  n10 = Rotator(Coordinate=TextureCoordinate()) {centerX=1, speed=0.0312}
  n11 = Rotator(Coordinate=TextureCoordinate()) {centerY=1, speed=-0.0312}
  n12 = Add(
   A=ComponentMask(None=TextureSample(UVs=@n10) {samplerType=SAMPLERTYPE_LinearColor, texture=/Game/Global/Textures/Default/Tileable_Patterns/T_Tile_Smoke_Small}.RGB) {b=true, g=true, r=true}
   B=ComponentMask(None=TextureSample(UVs=@n11) {samplerType=SAMPLERTYPE_LinearColor, texture=&T_Tile_Smoke_Small}.RGB) {b=true, g=true, r=true}
  )
  n13 = Add(A=ComponentMask(None=Multiply(A=@n12) {constB=0.1}) {g=true, r=true}, B=ScalarParameter() {defaultValue=0.15, parameterName=BaseDistort})
  n14 = TextureCoordinate()
  n15 = ShadingPathSwitch(Default=@n13, Mobile=@n14) {desc=Migrated from FeatureLevelSwitch}
  n16 = QualitySwitch(Default=@n13, Low=@n14, High=@n15, Medium=@n15)
  n17 = TextureSample(UVs=Add(A=AppendVector(A=@n7.Param1, B=@n7.Param1), B=Add(A=@n9, B=@n16))) {texture=&T_Inky_Smoke_Tile}
  n18 = AppendVector(A=@n7.Param1, B=@n7.Param1)
  n19 = Add(A=Panner(Coordinate=@n8) {speedX=-0.02, speedY=0.005}, B=@n16)
  n20 = Constant() {r=1}
  n21 = ShadingPathSwitch(Default=TextureSample(UVs=Add(A=@n18, B=@n19)) {texture=&T_Inky_Smoke_Tile}.G, Mobile=@n20) {desc=Migrated from FeatureLevelSwitch}
  n22 = QualitySwitch(Default=@n21, Low=@n20, High=@n21, Medium=@n21)
  n23 = Multiply(A=Multiply(
    A=TextureSample(UVs=@n6) {samplerType=SAMPLERTYPE_LinearColor, texture=/Game/Effects/Fort_Effects/Textures/Smoke/T_Smoke_8x8_01}.G
    B=Add(A=@n17.G, B=@n22)
   ))
  n24 = ParticleColor()
  n25 = Clamp(None=Multiply(A=Multiply(A=@n23, B=@n24.A), B=ParticleColor().A))
  n26 = ShadingPathSwitch(Default=Multiply(A=DepthFade() {fadeDistanceDefault=50}, B=@n25), Mobile=@n25) {desc=Migrated from FeatureLevelSwitch}
  QualitySwitch(Default=@n26, Low=@n25, High=@n26, Medium=@n26)
 ]
}
