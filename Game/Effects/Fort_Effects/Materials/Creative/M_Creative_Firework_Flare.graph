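# Unlit translucent particle material for Creative firework flare sprites, usable with Niagara and particle sprites.
# Particle color is hue-shifted by the RandomHue and Life dynamic parameters, dimmed over Life, then converted to HDR except on mobile.
# That color multiplies a glow built from powered radial masks plus a spiky FlareMask texture.
# Opacity combines particle alpha, a camera-distance fade and a circular mask, with depth fade skipped on mobile.
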Material /Game/Effects/Fort_Effects/Materials/Creative/M_Creative_Firework_Flare
graph {
 properties {
  allowTranslucentCustomDepthWrites=true
  bUsedWithNiagaraSprites=true
  bUsedWithParticleSprites=true
  blendMode=BLEND_Translucent
  opacityMaskClipValue=0.1
  pixelDepthOffsetMode=PDOM_Legacy
  shadingModel=MSM_Unlit
  usedShadingModels=Unlit
 }
 nodes [
  n1 = Saturate(None=DynamicParameter() {defaultValue=<rgba 1 1 -1 1>, paramNames [RandomHue,Life,OffsetMesh,LineWidth]}.Life)
  n2 = LinearInterpolate(Alpha=Saturate(None=DynamicParameter() {defaultValue=<rgba 1 1 -1 1>, paramNames [RandomHue,Life,OffsetMesh,LineWidth]}.RandomHue)) {constB=0.125}
  n3 = ParticleColor()
  n4 = Multiply(
   A=LinearInterpolate(Alpha=@n1) {constA=1, constB=0.75}
   B=MaterialFunctionCall(
    Hue Shift Percentage=LinearInterpolate(Alpha=@n1) {constB=0.25}
    Texture=MaterialFunctionCall(Hue Shift Percentage=@n2, Texture=@n3.RGB) {materialFunction=&HueShift}.Result
   ) {materialFunction=&HueShift}.Result
  )
  n5 = MaterialFunctionCall(LDR Input=Constant3Vector() {constant=<rgba 1 1 1 0>}, HDR Tint=@n4) {materialFunction=&DeriveHDRfromLDR}
  n6 = Constant() {r=0.5}
  n7 = SphereMask(A=TextureCoordinate(), B=@n6, Radius=@n6) {hardnessPercent=0}
  n8 = Multiply(A=Multiply(A=Power(Base=@n7) {constExponent=25}) {constB=2}, B=LinearInterpolate(B=Constant() {r=1}) {constA=1, constAlpha=0.7})
  n9 = Add(
   A=Multiply(A=Power(Base=@n7) {constExponent=4}) {constB=0.15}
   B=TextureSampleParameter2D() {parameterName=FlareMask, texture=/Game/Effects/Fort_Effects/Textures/Flares/T_Flare_Spikeball}.RGB
  )
  MaterialFunctionCall(Emissive=Multiply(A=ShadingPathSwitch(Default=@n5.RemappedRange, Mobile=@n4) {desc=Migrated from FeatureLevelSwitch}, B=Add(A=@n8, B=@n9))) {materialFunction=&MF_FX_EmissiveAdjustments}
  n10 = Power(Base=MaterialFunctionCall(Distance=ScalarParameter() {defaultValue=1024, parameterName=NearFade}) {materialFunction=&Near_Camera_Fade}.Result) {constExponent=4}
  n11 = Multiply(A=Multiply(A=Constant() {r=1}, B=ParticleColor().A), B=LinearInterpolate(A=Constant() {r=1}, B=Constant() {r=0.25}, Alpha=Clamp(None=@n10)))
  n12 = ScalarParameter() {defaultValue=512, parameterName=DepthFadeDistance}
  Multiply(
   A=SphereMask(
    A=TextureCoordinate()
    B=Constant2Vector() {g=0.5, r=0.5}
    Radius=ScalarParameter() {defaultValue=0.5, parameterName=Radius}
    Hardness=ScalarParameter() {parameterName=Hardness}
   )
   B=ShadingPathSwitch(Default=DepthFade(Opacity=@n11, FadeDistance=@n12), Mobile=@n11) {desc=Migrated from FeatureLevelSwitch}
  )
  Reroute(None=@n4)
 ]
}
