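# Unlit translucent particle sprite material for emissive smoke in Creative firework effects.
# Particle color is hue-shifted slightly by the RandomHue dynamic parameter, then shifted further and dimmed as Life rises.
# The color passes through DeriveHDRfromLDR on the default shading path but not on mobile.
# An 8x8 smoke texture's red channel masks emissive into MF_FX_EmissiveAdjustments, while green multiplies particle alpha for opacity.
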Material /Game/Effects/Fort_Effects/Materials/Creative/M_Creative_Firework_Smoke_Emissive
graph {
 properties {
  allowTranslucentCustomDepthWrites=true
  bEnableResponsiveAA=true
  bUsedWithParticleSprites=true
  blendMode=BLEND_Translucent
  opacityMaskClipValue=0.001
  pixelDepthOffsetMode=PDOM_Legacy
  shadingModel=MSM_Unlit
  translucencyLightingMode=TLM_VolumetricPerVertexDirectional
  usedShadingModels=Unlit
 }
 nodes [
  n1 = Saturate(None=DynamicParameter() {defaultValue=<rgba 1 1 -1 1>, paramNames [RandomHue,Life,OffsetMesh,LineWidth]}.Life)
  n2 = LinearInterpolate(Alpha=Saturate(None=DynamicParameter() {defaultValue=<rgba 1 1 -1 1>, paramNames [RandomHue,Life,OffsetMesh,LineWidth]}.RandomHue)) {constB=0.125}
  n3 = ParticleColor()
  n4 = Multiply(
   A=LinearInterpolate(Alpha=@n1) {constA=1, constB=0.75}
   B=MaterialFunctionCall(
    Hue Shift Percentage=LinearInterpolate(Alpha=@n1) {constB=0.25}
    Texture=MaterialFunctionCall(Hue Shift Percentage=@n2, Texture=@n3.RGB) {materialFunction=&HueShift}.Result
   ) {materialFunction=&HueShift}.Result
  )
  n5 = MaterialFunctionCall(LDR Input=Constant3Vector() {constant=<rgba 1 1 1 0>}, HDR Tint=@n4) {materialFunction=&DeriveHDRfromLDR}
  n6 = TextureSample() {samplerType=SAMPLERTYPE_LinearColor, texture=/Game/Effects/Fort_Effects/Textures/Smoke/T_Smoke_8x8_01}
  MaterialFunctionCall(Emissive=Multiply(A=ShadingPathSwitch(Default=@n5.RemappedRange, Mobile=@n4) {desc=Migrated from FeatureLevelSwitch}, B=@n6.R)) {materialFunction=&MF_FX_EmissiveAdjustments}
  Multiply(A=ParticleColor().A, B=@n6.G)
  ParticleColor()
 ]
}
